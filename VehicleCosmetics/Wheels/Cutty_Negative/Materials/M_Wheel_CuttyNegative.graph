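# Translucent unlit material for the Cutty Negative vehicle wheel, giving it an animated nebula look with optional paint tinting.
# Panning noise and twinkle textures form stars, combined with Fresnel, vertex-color and texture masks through named reroutes to drive emissive and opacity.
# World position offset combines radial displacement with a sine pulse.
# Output passes through damage and de-rez functions; a distortion term is built, but Refraction is left unconnected.
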Material /VehicleCosmetics/Wheels/Cutty_Negative/Materials/M_Wheel_CuttyNegative
graph {
 properties {
  bAutomaticallySetUsageInEditor=false
  bUseMaterialAttributes=true
  bUsedWithMeshDeformer=true
  bUsedWithSkeletalMesh=true
  blendMode=BLEND_Translucent
  previewMesh=/VehicleCosmetics/Wheels/Dinger_Inverted/New/Meshes/SK_Wheel_DingerInverted
  shadingModel=MSM_Unlit
  usedShadingModels=Unlit
 }
 nodes [
  n1 = SetMaterialAttributes!(MaterialAttributes, Emissive Color=NamedRerouteUsage(), Opacity=NamedRerouteUsage(), World Position Offset=NamedRerouteUsage(), Refraction) {attributeSetTypes [
    #1
    #2
    #3
    #4
   ]}
  MaterialFunctionCall(In=MaterialFunctionCall(Material=@n1) {materialFunction=&MF_BaseVehicle_Damage}.Result) {materialFunction=&MF_VehicleDeRez_Transition}
  NamedRerouteDeclaration(None=VertexColor().R) {name=VC_R, nodeColor=<rgba 0 1 0.0156 1>}
  n2 = NamedRerouteUsage()
  n3 = Constant() {r=8}
  n4 = Constant() {r=16}
  n5 = OneMinus(None=LinearInterpolate(
    A=MaterialFunctionCall!(Normal, Exponent=@n3) {materialFunction=/VehicleCosmetics/Wheels/Cola_Holographic/Materials/MF_Fresnel_Classic}.Result
    B=MaterialFunctionCall(Exponent=@n4) {materialFunction=&MF_Fresnel_Classic}.Result
    Alpha=NamedRerouteUsage()
   ))
  n6 = Multiply(A=Add(A=Multiply(A=@n2) {constB=0.25}, B=Subtract(A=@n5) {constB=0.9})) {constB=10}
  NamedRerouteDeclaration(None=OneMinus(None=Clamp(None=@n6))) {name=Fresnel, nodeColor=<rgba 1 0.0234 0 1>}
  n7 = TextureCoordinate() {coordinateIndex=2}
  n8 = Constant2Vector() {g=0.5, r=1}
  n9 = TextureSample(UVs=Panner(Coordinate=Multiply(A=@n7, B=@n8)) {speedX=0.05, speedY=-0.05}) {samplerSource=SSM_Wrap_WorldGroupSettings, texture=/VehicleCosmetics/Wheels/_SharedTextures/FX/Universal_Noise_55_RGB}
  n10 = Power(Base=@n9.R)
  n11 = Multiply(
   A=TextureSample(UVs=Panner(Coordinate=@n7) {speedX=0.1, speedY=-0.05}) {samplerSource=SSM_Wrap_WorldGroupSettings, texture=&Universal_Noise_55_RGB}.G
   B=@n9.G
  )
  n12 = ConstantBiasScale(None=OneMinus(None=NamedRerouteDeclaration(None=@n11) {name=Mix, nodeColor=<rgba 0.6641 0 1 1>})) {bias=-0.75, scale=2}
  n13 = Panner(Coordinate=Multiply(A=@n7, B=Constant2Vector() {g=0.75, r=1})) {speedX=0.25, speedY=-0.12}
  n14 = Power(Base=Multiply(A=Clamp(None=@n12), B=TextureSample(UVs=@n13) {texture=/VehicleCosmetics/Wheels/_SharedTextures/FX/skin_twankle2_rgb}.R))
  NamedRerouteDeclaration(None=Multiply(A=Max(A=@n10, B=@n14)) {constB=5}) {name=Stars, nodeColor=<rgba 0 0.2969 1 1>}
  n15 = Multiply(A=Subtract(A=WorldPosition().XYZ, B=ObjectPositionWS()), B=ObjectRadius())
  n16 = ScalarParameter() {defaultValue=-2, parameterName=WPO_int}
  n17 = Panner(Coordinate=TextureCoordinate() {coordinateIndex=2}) {speedX=1, speedY=1}
  n18 = Multiply(A=Sine(None=ComponentMask(None=@n17) {r=true})) {constB=300}
  n19 = Constant3Vector() {constant=<rgba 1 1 1 0>}
  n20 = Divide(A=Add(A=Multiply(A=@n15, B=@n16) {constB=-2}, B=Multiply(A=@n18, B=@n19) {constB=300})) {constB=1000}
  n21 = ScalarParameter() {parameterName=WPO Multiplier}
  NamedRerouteDeclaration(None=LinearInterpolate(B=Multiply(A=@n20, B=@n21), Alpha=NamedRerouteUsage())) {name=WPO, nodeColor=<rgba 0.1016 0 1 1>}
  n22 = NamedRerouteUsage()
  n23 = LinearInterpolate(A=LinearInterpolate(A=ScalarParameter() {defaultValue=2, parameterName=Opacity}, Alpha=Clamp(None=@n22)), Alpha=NamedRerouteUsage())
  n24 = Add(
   A=TextureSample() {samplerType=SAMPLERTYPE_LinearColor, texture=/VehicleCosmetics/Wheels/Cutty_Negative/Textures/T_CuttyNegative_SE_M_LOD}.R
   B=NamedRerouteUsage()
  )
  NamedRerouteDeclaration(None=StaticSwitchParameter(True=@n23, False=LinearInterpolate(B=@n23, Alpha=@n24)) {defaultValue=true, parameterName=Opacity}) {name=Opacity, nodeColor=<rgba 0 1 0.1797 1>}
  n25 = LinearInterpolate(
   A=Constant3Vector() {constant=<rgba 0.0375 0.0494 0.0829 0>}
   B=Multiply(
    A=VectorParameter() {parameterName=NebulaColor, defaultValue=<rgba 0.1529 0.0823 0.2051 1>}.RGB
    B=ScalarParameter() {defaultValue=1, parameterName=NebulaColorIntensity}
   )
   Alpha=NamedRerouteUsage()
  )
  n26 = VectorParameter() {parameterName=PaintColor, defaultValue=<rgba 1 0.7011 0.7011 1>}
  n27 = Multiply(A=Multiply(A=Desaturation(None=@n25) {luminanceFactors=<rgba 0.3 0.59 0.11 0>}) {constB=2}, B=@n26.RGB)
  n28 = Constant() {r=0.075}
  n29 = StaticSwitchParameter(
   True=LinearInterpolate(
    A=@n27
    B=MaterialFunctionCall!(Color=@n27, Angle=@n28) {materialFunction=/VehicleCosmetics/Wheels/Concept_Inverted/Materials/MF_HUE_Shift}.Result
    Alpha=@n27
   )
   False=@n25
  ) {parameterName=Painted}
  n30 = Multiply(A=@n26.RGB) {constB=6}
  n31 = NamedRerouteUsage()
  NamedRerouteDeclaration(None=LinearInterpolate(A=LinearInterpolate(A=@n29, B=@n30, Alpha=@n31), B=Multiply(A=@n26.RGB) {constB=2}, Alpha=NamedRerouteUsage())) {name=Emissive, nodeColor=<rgba 1 0 0.1406 1>}
  n32 = LinearInterpolate(Alpha=MaterialFunctionCall!(Input Normal, Input Sharpness=Constant() {r=1.5}, Input Two Sided) {materialFunction=/VehicleCosmetics/Wheels/CarCar_Holographic/Materials/MF_Fresnel_Advanced}.Edge) {constA=0.05, constB=0.4}
  n33 = ScalarParameter() {defaultValue=5, parameterName=Distortion_Edge_Intensity}
  n34 = TextureSample() {
   samplerSource=SSM_Wrap_WorldGroupSettings
   samplerType=SAMPLERTYPE_Normal
   texture=/VehicleCosmetics/Wheels/Cutty_Negative/Textures/T_CuttyNegative_Rim_N
  }
  n35 = TextureSample() {
   samplerSource=SSM_Wrap_WorldGroupSettings
   samplerType=SAMPLERTYPE_Normal
   texture=/VehicleCosmetics/Wheels/Cutty_Negative/Textures/T_CuttyNegative_Tire_N
  }
  n36 = LinearInterpolate(A=@n34.RGB, B=@n35.R, Alpha=NamedRerouteUsage())
  n37 = ScalarParameter() {defaultValue=2, parameterName=Distortion_Normal_Intensity}
  NamedRerouteDeclaration(None=Add(A=Multiply(A=@n32, B=@n33), B=Multiply(A=@n36, B=@n37))) {name=Dist, nodeColor=<rgba 1 0 0.7031 1>}
  NamedRerouteUsage()
  ScalarParameter() {defaultValue=1.33, parameterName=IOR}
 ]
}
