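# Masked, unlit LOD material for the rim of the Cutty Negative vehicle wheel cosmetic, built as animated emissive layers.
# Panned gradient textures, gated by packed masks, form flow and line glows, added to an interior rim color.
# A bump-offset scrolling noise layer, modulated by gradient sweeps and a sine pulse, blends dark/light intensities, tinted by inverted CustomColor unless Painted.
# Emission and opacity mask pass through shared vehicle damage and de-rez transition functions.
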Material /VehicleCosmetics/Wheels/Cutty_Negative/Materials/M_Wheel_CuttyNegative_Rim_LOD
graph {
 properties {
  bAutomaticallySetUsageInEditor=false
  bUseMaterialAttributes=true
  bUsedWithMeshDeformer=true
  bUsedWithSkeletalMesh=true
  blendMode=BLEND_Masked
  shadingModel=MSM_Unlit
  usedShadingModels=Unlit
 }
 nodes [
  Constant() {r=1}
  n1 = NamedRerouteUsage()
  n2 = Add(
   A=TextureSample() {samplerType=SAMPLERTYPE_LinearColor, texture=/VehicleCosmetics/Wheels/Cutty_Negative/Textures/T_CuttyNegative_Rim_M_LOD}.R
   B=NamedRerouteUsage()
  )
  MaterialFunctionCall(In=MaterialFunctionCall(Material=SetMaterialAttributes!(MaterialAttributes, Emissive Color=@n1, Opacity Mask=@n2) {attributeSetTypes [#1, #2]}) {materialFunction=&MF_BaseVehicle_Damage}.Result) {materialFunction=&MF_VehicleDeRez_Transition}
  n3 = TextureSampleParameter2D() {
   parameterName=Masks
   samplerType=SAMPLERTYPE_LinearColor
   texture=/VehicleCosmetics/Wheels/Cutty_Negative/Textures/T_CuttyNegative_Rim_RGB_LOD
  }
  Multiply(A=@n3.R)
  n4 = Constant() {r=1}
  n5 = Rotator(Coordinate=TextureCoordinate() {vTiling=0.5}, Time=@n4) {speed=0.5}
  n6 = NamedRerouteUsage()
  n7 = Multiply(
   A=TextureSample(UVs=Panner(Coordinate=@n5) {speedY=-0.15}, Tex=@n6) {samplerSource=SSM_Wrap_WorldGroupSettings, texture=/VehicleCosmetics/Wheels/Cutty_Negative/Textures/T_CuttyNegative_Rim_Gradients}.R
   B=@n3.A
  )
  n8 = Constant()
  n9 = NamedRerouteUsage()
  n10 = TextureCoordinate() {uTiling=0.5}
  n11 = Multiply(
   A=TextureSample(UVs=Panner(Coordinate=@n10) {speedX=-0.15}, Tex=@n6) {samplerSource=SSM_Wrap_WorldGroupSettings, texture=&T_CuttyNegative_Rim_Gradients}.G
   B=@n3.B
  )
  n12 = TextureSample(Tex=NamedRerouteUsage()) {samplerType=SAMPLERTYPE_LinearColor, texture=/VehicleCosmetics/Wheels/Cutty_Negative/Textures/T_CuttyNegative_Rim_RGB2}
  n13 = TextureCoordinate() {vTiling=0.1}
  n14 = TextureSample(UVs=Panner(Coordinate=Panner(Coordinate=@n13, Time=@n4) {speedY=0.48}) {speedY=-0.15}, Tex=@n6) {samplerSource=SSM_Wrap_WorldGroupSettings, texture=&T_CuttyNegative_Rim_Gradients}
  n15 = Add(
   A=LinearInterpolate(A=LinearInterpolate(A=@n7, B=@n8, Alpha=@n9), B=LinearInterpolate(B=@n11, Alpha=@n12.R), Alpha=@n9)
   B=LinearInterpolate(A=Multiply(A=@n14.R, B=@n12.A), B=@n8, Alpha=NamedRerouteUsage())
  )
  n16 = Clamp(None=Add(A=ScalarParameter() {defaultValue=0.2, parameterName=FlowDesaturate}, B=NamedRerouteUsage()))
  n17 = NamedRerouteUsage()
  n18 = NamedRerouteUsage()
  n19 = Constant() {r=10}
  n20 = LinearInterpolate(A=LinearInterpolate(A=@n3.G, B=Constant(), Alpha=Multiply(A=@n18, B=@n19)), B=@n3.R, Alpha=@n18)
  n21 = Add(
   A=Multiply(A=ScalarParameter() {defaultValue=4, parameterName=EmissiveFlowIntensity}, B=Multiply(A=@n15, B=@n16))
   B=Multiply(A=Multiply(A=@n17, B=@n20), B=ScalarParameter() {defaultValue=1.3, parameterName=EmissiveLineIntensity})
  )
  n22 = LinearInterpolate(
   A=Constant()
   B=Multiply(A=@n12.R, B=VectorParameter() {parameterName=InteriorRimColor, defaultValue=<rgba 0 0 0 1>}.RGB)
   Alpha=NamedRerouteUsage()
  )
  n23 = BumpOffset(Coordinate=TextureCoordinate() {coordinateIndex=1, uTiling=0.1, vTiling=2}, Height=Constant() {r=2}) {heightRatio=-0.2}
  n24 = Constant() {r=1}
  n25 = TextureSample(UVs=Panner(Coordinate=Panner(Coordinate=@n23, Time=@n24) {speedY=-0.5}) {speedX=0.25}) {texture=/VehicleCosmetics/Wheels/Cutty_Negative/Textures/T_CuttyNegative_Tire_Noise}
  n26 = Multiply(A=@n25.B)
  n27 = OneMinus(None=Add(A=@n25.G, B=@n25.G))
  n28 = OneMinus(None=TextureSample(UVs=TextureCoordinate() {coordinateIndex=2, unMirrorU=true}) {texture=/VehicleCosmetics/Wheels/_SharedTextures/FX/Gradiant_V_clamp}.R)
  n29 = TextureCoordinate() {coordinateIndex=2, uTiling=0.05}
  n30 = TextureObjectParameter() {parameterName=Gradients, texture=&T_CuttyNegative_Rim_Gradients}
  n31 = TextureSample(UVs=Panner(Coordinate=@n29) {speedX=0.15}, Tex=@n30) {samplerSource=SSM_Wrap_WorldGroupSettings, texture=&T_CuttyNegative_Rim_Gradients}
  n32 = Add(A=Multiply(A=Multiply(A=@n28) {constB=0.5}, B=@n31.B), B=@n31.B)
  n33 = Time()
  n34 = TextureSample(UVs=Panner(Coordinate=@n29, Time=Add(A=@n33) {constB=2.8}) {speedX=0.3}, Tex=@n30) {samplerSource=SSM_Wrap_WorldGroupSettings, texture=&T_CuttyNegative_Rim_Gradients}
  n35 = Multiply(A=Add(A=Sine(None=@n33) {period=0.15})) {constB=0.5}
  n36 = Multiply(A=@n32, B=Clamp(None=Add(A=@n34.B, B=@n35)))
  n37 = Multiply(A=LinearInterpolate(A=@n26, B=Min(A=@n27, B=@n26), Alpha=Clamp(None=@n36)), B=OneMinus(None=@n32))
  n38 = ScalarParameter() {parameterName=InvertLightIntensity}
  n39 = ScalarParameter() {defaultValue=1, parameterName=InvertDarkIntensity}
  n40 = OneMinus(None=NamedRerouteDeclaration(None=VectorParameter() {parameterName=CustomColor, defaultValue=<rgba 0.0482 0.4179 0.8228 1>}.RGB) {name=Custom Color, nodeColor=<rgba 1 0.1875 0 1>})
  n41 = StaticSwitchParameter(
   True=LinearInterpolate(A=@n38, B=@n39, Alpha=@n37)
   False=LinearInterpolate(A=Multiply(A=@n40, B=@n38), B=Multiply(A=@n40, B=@n39), Alpha=@n37)
  ) {parameterName=Painted}
  n42 = TextureObjectParameter() {
   parameterName=Masks UV
   samplerType=SAMPLERTYPE_LinearColor
   texture=&T_CuttyNegative_Rim_RGB2
  }
  n43 = TextureSample(UVs=TextureCoordinate() {coordinateIndex=1}, Tex=NamedRerouteDeclaration(None=@n42) {name=Masks UV, nodeColor=<rgba 0.8281 0 1 1>}) {
   samplerSource=SSM_Wrap_WorldGroupSettings
   samplerType=SAMPLERTYPE_LinearColor
   texture=&T_CuttyNegative_Rim_RGB2
  }
  NamedRerouteDeclaration(None=Add(A=Add(A=@n21, B=@n22), B=Multiply(A=@n41, B=@n43.R))) {name=Emission, nodeColor=<rgba 0.4297 0 1 1>}
  NamedRerouteDeclaration(None=@n30) {name=Gradients, nodeColor=<rgba 0.8516 1 0 1>}
  NamedRerouteDeclaration(None=VertexColor().R) {name=VC_R, nodeColor=<rgba 0 1 0.8125 1>}
  Reroute(None=@n3.G)
 ]
}
